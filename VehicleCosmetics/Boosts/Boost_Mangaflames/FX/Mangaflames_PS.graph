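# Niagara system for the Mangaflames vehicle cosmetic boost, a looping orange-red sprite flame trail.
# Three Game emitters build the trail from growing, fading sprites with random acceleration and inherited source movement, with a vortex on one; User.MultiplyScaleMain sets size and User.ScaleVelocityTrail backward speed.
# Two SingleLoopingParticle emitters add slower, brightened puffs that ignore both user parameters.
# All five animate SubUV frames and drive dynamic material parameters, mostly by curves over lifetime.
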NiagaraSystem /VehicleCosmetics/Boosts/Boost_Mangaflames/FX/Mangaflames_PS
dependencies {
 usedDataInterfaces [
  &NiagaraDataInterfaceCurve
  &NiagaraDataInterfaceVectorCurve
  &NiagaraDataInterfaceSpriteRendererInfo
 ]
 usedDynamicInputs []
 usedModules [&SystemState]
}
emitters {
 Game {
  emitterUpdateScript [
   &EmitterState(Life Cycle Mode=::NewEnumerator1, Loop Duration=1, Recalculate Duration Each Loop=true, UseLoopDelay=false)
   &SpawnRate(SpawnRate=30, Use Spawn Probability=false)
  ]
  particleSpawnScript [
   &InitializeParticle(
    Color=<rgba 1 0.2 0 1>
    Lifetime=&RandomRangeFloat(Minimum=0.8, Maximum=1)
    Sprite Rotation Angle=&RandomRangeFloat(Minimum=-1, Maximum=1)
    Sprite Rotation Mode=::NewEnumerator1
    Sprite Size Mode=::NewEnumerator3
    Uniform Sprite Size=&Multiply_Float(A=&RandomRangeFloat(Minimum=20, Maximum=10), B=$User.MultiplyScaleMain)
    UsePositionOffset=false
    Write Material Random=true
   )
   &AddVelocity(Velocity=<1000 0 0>, Velocity Speed Scale=$User.ScaleVelocityTrail)
   &AddVelocity(Velocity=&RandomRangeVector(Minimum=<0 50 50>, Maximum=<0 -50 -50>), Velocity Speed Scale=1.5) {module=AddVelocity001}
   &ShapeLocation(MaskSpawnGroup=false, Override Seed=false, Sphere Radius=&Multiply_Float(A=7, B=1), UseOwnerScale=true)
  ]
  particleUpdateScript [
   &ParticleState()
   &ScaleColor(
    Scale Alpha=&FloatFromCurve(FloatCurve=DataInterface() {propertyValues {Curve {keys [<at 0.5 1>,<at 0.8 0.1>,<at 1 0>]}}})
    ScaleA=true
    ScaleRGB=true
   )
   &ScaleSpriteSize(Uniform Curve Sprite Scale=DataInterface() {propertyValues {Curve {keys [<at 0 0.25>,<at 0.2 1>,<at 0.4 1.25>,<at 0.8 2>,<at 1.1 3>]}}})
   &ScaleVelocity(Velocity Scale=&VectorFromCurve(VectorCurve=DataInterface() {propertyValues {XCurve {keys [<at 0 0.6>,<at 0.45 0.5>,<at 1 0.1>]}, YCurve {keys [<at 0 0>,<at 0.4 0>,<at 1 0>]}, ZCurve {keys [<at 0 0>,<at 1 0>]}}}))
   &InheritSourceMovement(
    Applied Position Scale=&VectorFromCurve(VectorCurve=DataInterface() {propertyValues {XCurve {keys [<at 0 1>,<at 1 0>]}, YCurve {keys [<at 0 0>,<at 1 0>]}, ZCurve {keys [<at 0 0>,<at 1 0>]}}})
    ApplyForce=false
    ApplyPosition=true
    ApplyPositionFalloff=false
   )
   &VortexForce(
    UseOriginOffset=false
    Vortex Axis=&RandomRangeVector(Minimum=<0 0 0>, Maximum=<1 1 1>)
    Vortex Force Amount=500
   )
   &AccelerationForce(Acceleration=&Multiply_VectorByFloat(
     Vector=&RandomRangeVector(Minimum=<0 -300 -500>, Maximum=<1 300 200>)
     Float=&FloatFromCurve(FloatCurve=DataInterface() {propertyValues {Curve {keys [
          {arriveTangent=18.1997, interpMode=RCIM_Cubic, leaveTangent=18.1997, tangentMode=RCTM_User, time=0, value=1}
          {arriveTangent=-0.8133, interpMode=RCIM_Cubic, leaveTangent=-0.8133, tangentMode=RCTM_User, time=1, value=15}
         ]}}})
    ))
   &SolveForcesAndVelocity(Clamp Velocity=false, Limit Acceleration=false)
   &DynamicMaterialParameters(
    Index 0 Param 1=6
    Index 0 Param 2=&FloatFromCurve(FloatCurve=DataInterface() {propertyValues {Curve {keys [
         <at 0 0 RCIM_Cubic>
         {arriveTangent=1.4286, interpMode=RCIM_Cubic, leaveTangent=1.4286, time=0.4, value=1.5}
         {arriveTangent=-0.0117, interpMode=RCIM_Cubic, leaveTangent=-0.0117, tangentMode=RCTM_User, time=0.7, value=1}
         <at 1 0 RCIM_Cubic>
        ]}}})
    Index 0 Param 3=&FloatFromCurve(FloatCurve=DataInterface() {propertyValues {Curve {keys [
         <at 0 1 RCIM_Cubic>
         {interpMode=RCIM_Cubic, tangentMode=RCTM_User, time=0.4, value=0}
         {interpMode=RCIM_Cubic, tangentMode=RCTM_User, time=0.7, value=0}
         <at 1 1 RCIM_Cubic>
        ]}}})
    Index 0 Param 4=6
    Module.Index0Param0Write=true
    Module.Index0Param1Write=true
    Module.Index0Param2Write=true
    Module.Index0Param3Write=true
   )
   &SubUVAnimation(
    "Sprite Renderer "=DataInterface() {propertyValues {SpriteRenderer=&self:Game_4.NiagaraSpriteRendererProperties_0}}
    SubUV Animation Mode:ENiagara_SubUVLookupModeV2=::NewEnumerator1
    SubUV Curve=DataInterface() {propertyValues {Curve {keys [{tangentMode=RCTM_User, time=0, value=0}, {tangentMode=RCTM_User, time=1, value=4}]}}}
    UseEndFrame=false
    UseStartFrame=false
   )
   &ShapeLocation(MaskSpawnGroup=false, Override Seed=false, Sphere Radius=&Multiply_Float(A=25.3204, B=2), UseOwnerScale=true) {enabled=false, module=ShapeLocation001}
  ]
  rendererClasses [&NiagaraSpriteRendererProperties]
  renderers [{rendererClass=&NiagaraSpriteRendererProperties}]
 }
 Game02 {
  emitterUpdateScript [
   &EmitterState(Life Cycle Mode=::NewEnumerator1, Loop Duration=1, Recalculate Duration Each Loop=true, UseLoopDelay=false)
   &SpawnRate(SpawnRate=60, Use Spawn Probability=false)
  ]
  particleSpawnScript [
   &InitializeParticle(
    Color=<rgba 1 0.2 0.01 1>
    Lifetime=&RandomRangeFloat(Minimum=0.8, Maximum=1)
    Position Offset=<-30 0 0>
    Sprite Rotation Angle=&RandomRangeFloat(Minimum=-1, Maximum=1)
    Sprite Rotation Mode=::NewEnumerator1
    Sprite Size Mode=::NewEnumerator3
    Uniform Sprite Size=&Multiply_Float(A=30, B=$User.MultiplyScaleMain)
    UsePositionOffset=true
    Write Material Random=true
   )
   &AddVelocity(Velocity=<1000 0 0>, Velocity Speed Scale=$User.ScaleVelocityTrail)
   &AddVelocity(Velocity=&RandomRangeVector(Minimum=<0 50 50>, Maximum=<0 -50 -50>), Velocity Speed Scale=1.5) {module=AddVelocity001}
   &ShapeLocation(MaskSpawnGroup=false, Override Seed=false, Sphere Radius=&Multiply_Float(A=6, B=1), UseOwnerScale=true)
  ]
  particleUpdateScript [
   &ParticleState()
   &SpriteRotationRate(Rotation Rate=&RandomRangeFloat(Minimum=-0.25, Maximum=0.25))
   &ScaleColor(
    Scale Alpha=&FloatFromCurve(FloatCurve=DataInterface() {propertyValues {Curve {keys [<at 0.25 1>,<at 0.3 1>,<at 0.6 0.4>,<at 1 0>]}}})
    ScaleA=true
    ScaleRGB=true
   )
   &ScaleSpriteSize(Uniform Curve Sprite Scale=DataInterface() {propertyValues {Curve {keys [<at 0 0.25>,<at 0.2 1>,<at 0.4 1.25>,<at 0.8 2>,<at 1.1 3>]}}})
   &ScaleVelocity(Velocity Scale=&VectorFromCurve(VectorCurve=DataInterface() {propertyValues {XCurve {keys [<at 0 0>,<at 0.5 0.8>,<at 1 0.15>]}, YCurve {keys [<at 0 0>,<at 0.4 0>,<at 1 0>]}, ZCurve {keys [<at 0 0>,<at 1 0>]}}}))
   &InheritSourceMovement(
    Applied Position Scale=&VectorFromCurve(VectorCurve=DataInterface() {propertyValues {XCurve {keys [<at 0 1>,<at 1 0>]}, YCurve {keys [<at 0 0>,<at 1 0>]}, ZCurve {keys [<at 0 0>,<at 1 0>]}}})
    ApplyForce=false
    ApplyPosition=true
    ApplyPositionFalloff=false
   )
   &AccelerationForce(Acceleration=&Multiply_VectorByFloat(
     Vector=&RandomRangeVector(Minimum=<0 -300 -501>, Maximum=<0 300 200>)
     Float=&FloatFromCurve(FloatCurve=DataInterface() {propertyValues {Curve {keys [
          {arriveTangent=18.1997, interpMode=RCIM_Cubic, leaveTangent=18.1997, tangentMode=RCTM_User, time=0, value=1}
          {arriveTangent=-0.8133, interpMode=RCIM_Cubic, leaveTangent=-0.8133, tangentMode=RCTM_User, time=1, value=15}
         ]}}})
    ))
   &SolveForcesAndVelocity(Clamp Velocity=false, Limit Acceleration=false)
   &DynamicMaterialParameters(
    Index 0 Param 1=8
    Index 0 Param 2=&FloatFromCurve(FloatCurve=DataInterface() {propertyValues {Curve {keys [
         <at 0 0 RCIM_Cubic>
         {arriveTangent=1.4286, interpMode=RCIM_Cubic, leaveTangent=1.4286, time=0.4, value=1.5}
         {arriveTangent=-0.0117, interpMode=RCIM_Cubic, leaveTangent=-0.0117, tangentMode=RCTM_User, time=0.7, value=1}
         <at 1 0 RCIM_Cubic>
        ]}}})
    Index 0 Param 3=&FloatFromCurve(FloatCurve=DataInterface() {propertyValues {Curve {keys [
         <at 0 1 RCIM_Cubic>
         {interpMode=RCIM_Cubic, tangentMode=RCTM_User, time=0.4, value=0}
         {interpMode=RCIM_Cubic, tangentMode=RCTM_User, time=0.7, value=0}
         <at 1 1 RCIM_Cubic>
        ]}}})
    Index 0 Param 4=6
    Module.Index0Param0Write=true
    Module.Index0Param1Write=true
    Module.Index0Param2Write=true
    Module.Index0Param3Write=true
   )
   &SubUVAnimation(
    "Sprite Renderer "=DataInterface() {propertyValues {SpriteRenderer=&self:Game02_7.NiagaraSpriteRendererProperties_0}}
    SubUV Animation Mode:ENiagara_SubUVLookupModeV2=::NewEnumerator1
    SubUV Curve=DataInterface() {propertyValues {Curve {keys [{tangentMode=RCTM_User, time=0, value=0}, {tangentMode=RCTM_User, time=1, value=4}]}}}
    UseEndFrame=false
    UseStartFrame=false
   )
   &ShapeLocation(
    MaskSpawnGroup=false
    Override Seed=false
    Shape Origin=$StackContext.Position
    Sphere Radius=&Multiply_Float(A=5, B=15)
    UseOwnerScale=true
   ) {enabled=false, module=ShapeLocation001}
  ]
  rendererClasses [&NiagaraSpriteRendererProperties]
  renderers [{rendererClass=&NiagaraSpriteRendererProperties}]
 }
 Game03 {
  emitterUpdateScript [
   &EmitterState(Life Cycle Mode=::NewEnumerator1, Loop Duration=1, Recalculate Duration Each Loop=true, UseLoopDelay=false)
   &SpawnRate(SpawnRate=60, Use Spawn Probability=false)
  ]
  particleSpawnScript [
   &InitializeParticle(
    Color=<rgba 1 0.1 0 1>
    Lifetime=&RandomRangeFloat(Minimum=0.8, Maximum=1)
    Position Offset=<-30 0 0>
    Sprite Rotation Angle=&RandomRangeFloat(Minimum=-1, Maximum=1)
    Sprite Rotation Mode=::NewEnumerator1
    Sprite Size Mode=::NewEnumerator3
    Uniform Sprite Size=&Multiply_Float(A=30, B=$User.MultiplyScaleMain)
    UsePositionOffset=true
    Write Material Random=true
   )
   &AddVelocity(Velocity=<1000 0 0>, Velocity Speed Scale=$User.ScaleVelocityTrail)
   &AddVelocity(Velocity=&RandomRangeVector(Minimum=<0 50 50>, Maximum=<0 -50 -50>), Velocity Speed Scale=1.5) {module=AddVelocity001}
   &ShapeLocation(MaskSpawnGroup=false, Override Seed=false, Sphere Radius=&Multiply_Float(A=6, B=1), UseOwnerScale=true)
  ]
  particleUpdateScript [
   &ParticleState()
   &SpriteRotationRate(Rotation Rate=&RandomRangeFloat(Minimum=-0.25, Maximum=0.25))
   &ScaleColor(
    Scale Alpha=&FloatFromCurve(FloatCurve=DataInterface() {propertyValues {Curve {keys [<at 0.25 1>,<at 0.3 1>,<at 0.6 0.45>,<at 1 0>]}}})
    Scale RGB=&Multiply_VectorByFloat(Vector=<1 1 1>, Float=0.75)
    ScaleA=true
    ScaleRGB=true
   )
   &ScaleSpriteSize(Uniform Curve Sprite Scale=DataInterface() {propertyValues {Curve {keys [<at 0 0.25>,<at 0.2 1>,<at 0.4 1.25>,<at 0.8 2>,<at 1.1 3>]}}})
   &ScaleVelocity(Velocity Scale=&VectorFromCurve(VectorCurve=DataInterface() {propertyValues {XCurve {keys [<at 0 0>,<at 0.5 0.8>,<at 1 0.15>]}, YCurve {keys [<at 0 0>,<at 0.4 0>,<at 1 0>]}, ZCurve {keys [<at 0 0>,<at 1 0>]}}}))
   &InheritSourceMovement(
    Applied Position Scale=&VectorFromCurve(VectorCurve=DataInterface() {propertyValues {XCurve {keys [<at 0 1>,<at 1 0>]}, YCurve {keys [<at 0 0>,<at 1 0>]}, ZCurve {keys [<at 0 0>,<at 1 0>]}}})
    ApplyForce=false
    ApplyPosition=true
    ApplyPositionFalloff=false
   )
   &AccelerationForce(Acceleration=&Multiply_VectorByFloat(
     Vector=&RandomRangeVector(Minimum=<0 -300 -500>, Maximum=<0 300 200>)
     Float=&FloatFromCurve(FloatCurve=DataInterface() {propertyValues {Curve {keys [
          {arriveTangent=18.1997, interpMode=RCIM_Cubic, leaveTangent=18.1997, tangentMode=RCTM_User, time=0, value=1}
          {arriveTangent=-0.8133, interpMode=RCIM_Cubic, leaveTangent=-0.8133, tangentMode=RCTM_User, time=1, value=15}
         ]}}})
    ))
   &SolveForcesAndVelocity(Clamp Velocity=false, Limit Acceleration=false)
   &DynamicMaterialParameters(
    Index 0 Param 1=8
    Index 0 Param 2=&FloatFromCurve(FloatCurve=DataInterface() {propertyValues {Curve {keys [
         <at 0 0 RCIM_Cubic>
         {arriveTangent=1.4286, interpMode=RCIM_Cubic, leaveTangent=1.4286, time=0.4, value=1.5}
         {arriveTangent=-0.0117, interpMode=RCIM_Cubic, leaveTangent=-0.0117, tangentMode=RCTM_User, time=0.7, value=1}
         <at 1 0 RCIM_Cubic>
        ]}}})
    Index 0 Param 3=&FloatFromCurve(FloatCurve=DataInterface() {propertyValues {Curve {keys [
         <at 0 1 RCIM_Cubic>
         {interpMode=RCIM_Cubic, tangentMode=RCTM_User, time=0.4, value=0}
         {interpMode=RCIM_Cubic, tangentMode=RCTM_User, time=0.7, value=0}
         <at 1 1 RCIM_Cubic>
        ]}}})
    Index 0 Param 4=8
    Module.Index0Param0Write=true
    Module.Index0Param1Write=true
    Module.Index0Param2Write=true
    Module.Index0Param3Write=true
   )
   &SubUVAnimation(
    "Sprite Renderer "=DataInterface() {propertyValues {SpriteRenderer=&self:Game03_10.NiagaraSpriteRendererProperties_0}}
    SubUV Animation Mode:ENiagara_SubUVLookupModeV2=::NewEnumerator1
    SubUV Curve=DataInterface() {propertyValues {Curve {keys [{tangentMode=RCTM_User, time=0, value=0}, {tangentMode=RCTM_User, time=1, value=4}]}}}
    UseEndFrame=false
    UseStartFrame=false
   )
  ]
  rendererClasses [&NiagaraSpriteRendererProperties]
  renderers [{rendererClass=&NiagaraSpriteRendererProperties}]
 }
 SingleLoopingParticle {
  emitterUpdateScript [
   &EmitterState(Life Cycle Mode=::NewEnumerator1, Loop Duration=6, Recalculate Duration Each Loop=true, UseLoopDelay=false)
   &SpawnRate(SpawnRate=16, Use Spawn Probability=false)
   &SpawnBurst_Instantaneous(Spawn Count=2, Use Spawn Probability=false, UseLoopCountLimit=false) {enabled=false}
   &SpawnBurst_Instantaneous(Spawn Time=0.2, Use Spawn Probability=false, UseLoopCountLimit=false) {enabled=false, module=SpawnBurst_Instantaneous001}
   &SpawnBurst_Instantaneous(Spawn Count=2, Spawn Time=2.5, Use Spawn Probability=false, UseLoopCountLimit=false) {enabled=false, module=SpawnBurst_Instantaneous002}
   &SpawnBurst_Instantaneous(Spawn Time=2.7, Use Spawn Probability=false, UseLoopCountLimit=false) {enabled=false, module=SpawnBurst_Instantaneous003}
   &SpawnBurst_Instantaneous(Spawn Count=2, Spawn Time=4, Use Spawn Probability=false, UseLoopCountLimit=false) {enabled=false, module=SpawnBurst_Instantaneous004}
   &SpawnBurst_Instantaneous(Spawn Count=2, Spawn Time=4.2, Use Spawn Probability=false, UseLoopCountLimit=false) {enabled=false, module=SpawnBurst_Instantaneous005}
  ]
  particleSpawnScript [
   &InitializeParticle(
    Color=<rgba 1 0.2 0.01 1>
    Lifetime=&RandomRangeFloat(Minimum=0.75, Maximum=1)
    Sprite Rotation Angle=&RandomRangeFloat(Minimum=-1, Maximum=1)
    Sprite Rotation Mode=::NewEnumerator1
    Sprite Size Mode=::NewEnumerator3
    Uniform Sprite Size=&Multiply_Float(A=&RandomRangeFloat(Minimum=2, Maximum=3), B=1.2)
    UsePositionOffset=false
    Write Material Random=true
   )
   &AddVelocity(Velocity=<100 0 0>, Velocity Speed Scale=-1)
   &AddVelocity(Velocity=&RandomRangeVector(Minimum=<-55 15 15>, Maximum=<-45 -15 -15>)) {module=AddVelocity001}
  ]
  particleUpdateScript [
   &ParticleState()
   &ScaleColor(
    Scale Alpha=&FloatFromCurve(FloatCurve=DataInterface() {propertyValues {Curve {keys [<at 0 1>,<at 0.5 0>]}}})
    Scale RGB=<2 2 2>
    ScaleA=true
    ScaleRGB=true
   )
   &ScaleSpriteSize(Uniform Curve Sprite Scale=DataInterface() {propertyValues {Curve {keys [<at 0 10 RCIM_Cubic>,<at 1 45 RCIM_Cubic>]}}})
   &SpriteRotationRate(Rotation Rate=&RandomRangeFloat(Minimum=-0.1, Maximum=0.1))
   &ScaleVelocity(Velocity Scale=&VectorFromCurve(VectorCurve=DataInterface() {propertyValues {
       XCurve {keys [<at 0 0.6>,<at 0.45 0.5>,<at 1 0.1>]}
       YCurve {keys [<at 0 0>,<at 0.45 0>,<at 1 0>]}
       ZCurve {keys [<at 0 0>,<at 0.45 0>,<at 1 0>]}
      }}))
   &SolveForcesAndVelocity(Clamp Velocity=false, Limit Acceleration=false)
   &DynamicMaterialParameters(
    Index 0 Param 1=&FloatFromCurve(FloatCurve=DataInterface() {propertyValues {Curve {keys [<at 0 35>,<at 0.15 4>]}}})
    Index 0 Param 2=&FloatFromCurve(FloatCurve=DataInterface() {propertyValues {Curve {keys [<at 0 0.1>,<at 0.8 0.5>]}}})
    Index 0 Param 3=3
    Index 0 Param 4=2
    Module.Index0Param0Write=true
    Module.Index0Param1Write=true
    Module.Index0Param2Write=true
    Module.Index0Param3Write=true
   )
   &SubUVAnimation(
    "Sprite Renderer "=DataInterface() {propertyValues {SpriteRenderer=&self:SingleLoopingParticle.NiagaraSpriteRendererProperties_0}}
    SubUV Animation Mode:ENiagara_SubUVLookupModeV2=::NewEnumerator1
    SubUV Curve=DataInterface() {propertyValues {Curve {keys [{tangentMode=RCTM_User, time=0, value=0}, {tangentMode=RCTM_User, time=1, value=4}]}}}
    UseEndFrame=false
    UseStartFrame=false
   )
  ]
  rendererClasses [&NiagaraSpriteRendererProperties]
  renderers [{rendererClass=&NiagaraSpriteRendererProperties}]
 }
 SingleLoopingParticle001 {
  emitterUpdateScript [
   &EmitterState(Life Cycle Mode=::NewEnumerator1, Loop Duration=6, Recalculate Duration Each Loop=true, UseLoopDelay=false)
   &SpawnRate(SpawnRate=16, Use Spawn Probability=false)
   &SpawnBurst_Instantaneous(Spawn Count=2, Use Spawn Probability=false, UseLoopCountLimit=false) {enabled=false}
   &SpawnBurst_Instantaneous(Spawn Time=0.2, Use Spawn Probability=false, UseLoopCountLimit=false) {enabled=false, module=SpawnBurst_Instantaneous001}
   &SpawnBurst_Instantaneous(Spawn Count=2, Spawn Time=2.5, Use Spawn Probability=false, UseLoopCountLimit=false) {enabled=false, module=SpawnBurst_Instantaneous002}
   &SpawnBurst_Instantaneous(Spawn Time=2.7, Use Spawn Probability=false, UseLoopCountLimit=false) {enabled=false, module=SpawnBurst_Instantaneous003}
   &SpawnBurst_Instantaneous(Spawn Count=2, Spawn Time=4, Use Spawn Probability=false, UseLoopCountLimit=false) {enabled=false, module=SpawnBurst_Instantaneous004}
   &SpawnBurst_Instantaneous(Spawn Count=2, Spawn Time=4.2, Use Spawn Probability=false, UseLoopCountLimit=false) {enabled=false, module=SpawnBurst_Instantaneous005}
  ]
  particleSpawnScript [
   &InitializeParticle(
    Color=<rgba 1 0.2 0.01 1>
    Lifetime=&RandomRangeFloat(Minimum=0.8, Maximum=1.5)
    Sprite Rotation Angle=&RandomRangeFloat(Minimum=-1, Maximum=1)
    Sprite Rotation Mode=::NewEnumerator1
    Sprite Size Mode=::NewEnumerator3
    Uniform Sprite Size=&Multiply_Float(A=&RandomRangeFloat(Minimum=1, Maximum=0.5), B=1.2)
    UsePositionOffset=false
    Write Material Random=true
   )
   &AddVelocity(Velocity=<100 0 0>, Velocity Speed Scale=-1)
   &AddVelocity(Velocity=&RandomRangeVector(Minimum=<-50 15 15>, Maximum=<-25 -15 -15>)) {module=AddVelocity001}
   &ShapeLocation(
    MaskSpawnGroup=false
    Override Seed=false
    Sphere Radius=&Multiply_FloatByInt(Float=10, Integer=1)
    UseOwnerScale=true
   )
  ]
  particleUpdateScript [
   &ParticleState()
   &ScaleColor(
    Scale Alpha=&FloatFromCurve(FloatCurve=DataInterface() {propertyValues {Curve {keys [<at 0 1>,<at 0.5 0>]}}})
    Scale RGB=<2 2 2>
    ScaleA=true
    ScaleRGB=true
   )
   &ScaleSpriteSize(Uniform Curve Sprite Scale=DataInterface() {propertyValues {Curve {keys [<at 0 10 RCIM_Cubic>,<at 1 45 RCIM_Cubic>]}}})
   &SpriteRotationRate(Rotation Rate=&RandomRangeFloat(Minimum=-0.1, Maximum=0.1))
   &ScaleVelocity(Velocity Scale=&VectorFromCurve(VectorCurve=DataInterface() {propertyValues {
       XCurve {keys [<at 0 0.6>,<at 0.45 0.5>,<at 1 0.1>]}
       YCurve {keys [<at 0 0>,<at 0.45 0>,<at 1 0>]}
       ZCurve {keys [<at 0 0>,<at 0.45 0>,<at 1 0>]}
      }}))
   &SolveForcesAndVelocity(Clamp Velocity=false, Limit Acceleration=false)
   &DynamicMaterialParameters(
    Index 0 Param 1=&FloatFromCurve(FloatCurve=DataInterface() {propertyValues {Curve {keys [<at 0 35>,<at 0.15 4>]}}})
    Index 0 Param 2=&FloatFromCurve(FloatCurve=DataInterface() {propertyValues {Curve {keys [<at 0 0.1>,<at 0.8 0.2>]}}})
    Index 0 Param 3=3
    Index 0 Param 4=2
    Module.Index0Param0Write=true
    Module.Index0Param1Write=true
    Module.Index0Param2Write=true
    Module.Index0Param3Write=true
   )
   &SubUVAnimation(
    "Sprite Renderer "=DataInterface() {propertyValues {SpriteRenderer=&self:SingleLoopingParticle001_1.NiagaraSpriteRendererProperties_0}}
    SubUV Animation Mode:ENiagara_SubUVLookupModeV2=::NewEnumerator1
    SubUV Curve=DataInterface() {propertyValues {Curve {keys [{tangentMode=RCTM_User, time=0, value=0}, {tangentMode=RCTM_User, time=1, value=4}]}}}
    UseEndFrame=false
    UseStartFrame=false
   )
  ]
  rendererClasses [&NiagaraSpriteRendererProperties]
  renderers [{rendererClass=&NiagaraSpriteRendererProperties}]
 }
}
summary {
 systemName=Mangaflames_PS
 emitters [
  {emitterName=SingleLoopingParticle, rendererClasses [&NiagaraSpriteRendererProperties]}
  {emitterName=SingleLoopingParticle001, rendererClasses [&NiagaraSpriteRendererProperties]}
  {emitterName=Game, rendererClasses [&NiagaraSpriteRendererProperties]}
  {emitterName=Game02, rendererClasses [&NiagaraSpriteRendererProperties]}
  {emitterName=Game03, rendererClasses [&NiagaraSpriteRendererProperties]}
 ]
 userVariables [
  User.MultiplyScaleMain {type=NiagaraFloat, defaultValue=1, description=""}
  User.ScaleVelocityTrail {type=NiagaraFloat, defaultValue=-0.9, description=""}
 ]
}
