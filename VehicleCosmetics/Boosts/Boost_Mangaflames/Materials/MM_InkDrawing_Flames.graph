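# Unlit masked Niagara sprite material for ink-drawn, manga-style flames in a vehicle boost cosmetic.
# A sketch-smoke texture's green channel splits into fill and line, shaped and lit by dynamic parameters and tinted by particle color.
# Texture alpha is multiplied by a sine of distortion noise, tiled radial gradient and particle alpha, forming a dissolve mask that also adds glow.
# TextureSwap blends identical inputs, so has no effect; the flow texture is unused.
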Material /VehicleCosmetics/Boosts/Boost_Mangaflames/Materials/MM_InkDrawing_Flames
graph {
 properties {
  bUsedWithNiagaraSprites=true
  blendMode=BLEND_Masked
  pixelDepthOffsetMode=PDOM_Legacy
  refractionMethod=RM_PixelNormalOffset
  shadingModel=MSM_Unlit
  usedShadingModels=Unlit
 }
 nodes [
  n1 = DynamicParameter() {paramNames [Brightness, EdgeThickness, TextureSwap, LineBrightness]}
  n2 = TextureSample(UVs=TextureCoordinate()) {samplerType=SAMPLERTYPE_LinearColor, texture=/VehicleCosmetics/Boosts/Boost_Mangaflames/Textures/T_SketchSmoke_FB_Alpha}
  n3 = Saturate(None=@n1.TextureSwap)
  n4 = LinearInterpolate(A=@n2.G, B=@n2.G, Alpha=@n3)
  n5 = Saturate(None=ConstantBiasScale(None=Multiply(A=@n4, B=@n4)) {bias=-0.6, scale=3})
  n6 = Saturate(None=Multiply(A=@n4) {constB=600})
  n7 = LinearInterpolate(A=Multiply(A=@n5, B=@n5), B=Multiply(A=@n6, B=@n6), Alpha=@n1.EdgeThickness)
  n8 = Saturate(None=OneMinus(None=@n7))
  n9 = ParticleColor()
  n10 = Desaturation(None=@n9.RGB, Fraction=Constant() {r=0.1})
  n11 = LinearInterpolate(
   A=Multiply(A=Multiply(A=@n1.Brightness, B=@n8), B=Multiply(A=@n10, B=@n9.RGB))
   B=Multiply(A=Multiply(A=@n1.LineBrightness, B=@n7), B=@n10)
   Alpha=Constant() {r=0.5}
  )
  n12 = LinearInterpolate(A=@n2.A, B=@n2.A, Alpha=@n3)
  n13 = TextureCoordinate() {uTiling=2, vTiling=2}
  n14 = ConstantBiasScale(None=Desaturation(None=TextureSample(UVs=@n13) {texture=/VehicleCosmetics/Boosts/Boost_Mangaflames/Textures/T_Distort_RGB}.RGB)) {bias=0.1, scale=1}
  n15 = Add(A=Frac(None=@n13), B=Constant() {r=-0.5})
  n16 = DotProduct(A=@n15, B=@n15)
  n17 = Saturate(None=ConstantBiasScale(None=OneMinus(None=@n16)) {bias=-0.8, scale=1})
  n18 = Saturate(None=Multiply(A=@n17, B=@n17))
  n19 = ParticleColor()
  n20 = Add(A=Saturate(None=Add(A=@n14, B=@n18)), B=OneMinus(None=Add(A=@n19.A) {constB=0.2}))
  n21 = Multiply(A=Saturate(None=Multiply(A=@n12) {constB=16}), B=Saturate(None=Sine(None=@n20) {period=2}))
  n22 = Saturate(None=Multiply(A=@n9.RGB, B=@n21))
  MaterialFunctionCall(Emissive=Saturate(None=Add(A=@n11, B=@n22))) {materialFunction=&MF_FX_EmissiveAdjustments}
  Saturate(None=@n21)
  TextureSample(UVs=TextureCoordinate() {uTiling=2, vTiling=2}) {texture=/VehicleCosmetics/Boosts/Boost_Mangaflames/Textures/T_Flow_02}
 ]
}
