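# Translucent unlit Niagara sprite material for stylized flames in the Mangaflames vehicle boost.
# Two opposing panned samples of a flow normal map, timed by dynamic Speed, distort the UVs of a sketch-smoke texture by DistortAmount.
# A near-binary red channel plus dimmed blue forms the glow, scaled by Brightness and particle color through an emissive adjustment function.
# Particle alpha erodes the blue channel, with Hardness sharpening the resulting mask.
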Material /VehicleCosmetics/Boosts/Boost_Mangaflames/Materials/MM_flame
graph {
 properties {
  bUsedWithNiagaraSprites=true
  blendMode=BLEND_Translucent
  pixelDepthOffsetMode=PDOM_Legacy
  refractionMethod=RM_PixelNormalOffset
  shadingModel=MSM_Unlit
  usedShadingModels=Unlit
 }
 nodes [
  n1 = Constant() {r=1.3}
  n2 = DynamicParameter() {paramNames [Hardness, DistortAmount, Brightness, Speed]}
  n3 = MaterialFunctionCall(Speed=@n2.Speed) {materialFunction=&TimeWithSpeedVariable}
  n4 = Panner(Coordinate=Multiply(A=TextureCoordinate(), B=@n1), Time=@n3.Time) {speedX=0.2, speedY=0.1}
  n5 = Panner(Coordinate=Multiply(A=TextureCoordinate() {uTiling=0.5, vTiling=0.5}, B=@n1), Time=@n3.Time) {speedX=-0.12, speedY=-0.2}
  n6 = ComponentMask(None=Add(
    A=TextureSample(UVs=@n4) {samplerType=SAMPLERTYPE_Normal, texture=/VehicleCosmetics/Boosts/Boost_Mangaflames/Textures/T_Flow_03}.RGB
    B=TextureSample(UVs=@n5) {samplerType=SAMPLERTYPE_Normal, texture=&T_Flow_03}.RGB
   )) {g=true, r=true}
  n7 = TextureSample(UVs=Add(A=TextureCoordinate(), B=Multiply(A=@n6, B=@n2.DistortAmount))) {samplerType=SAMPLERTYPE_LinearColor, texture=/VehicleCosmetics/Boosts/Boost_Mangaflames/Textures/T_SketchSmoke_FB_Alpha}
  n8 = Saturate(None=Multiply(A=@n7.R) {constB=1000})
  n9 = Constant() {r=0.3}
  n10 = Saturate(None=Add(A=Multiply(A=@n8) {constB=2}, B=Multiply(A=@n7.B, B=@n9)))
  n11 = ParticleColor()
  MaterialFunctionCall(Emissive=Multiply(A=Multiply(A=@n10, B=@n2.Brightness), B=@n11.RGB)) {materialFunction=&MF_FX_EmissiveAdjustments}
  n12 = Saturate(None=@n7.B)
  n13 = OneMinus(None=@n11.A)
  Saturate(None=Multiply(A=@n2.Hardness, B=Subtract(A=@n12, B=@n13)))
 ]
}
